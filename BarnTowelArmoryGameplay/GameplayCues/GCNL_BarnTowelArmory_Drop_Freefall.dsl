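; Looping gameplay cue for the BarnTowel Armory drop freefall that adds a special music track while the cue is active.
; On looping start, after parent logic, it spawns the freefall music attached to the cue's first particle component and stores the audio component.
; It registers that component with the target's MusicEventSubsystem under the POI Special event tag.
; On removal, it removes that music event if the stored component is valid.
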(event EventOnLoopingStart(EitherVFX-preferthisversion)
  (params (MyTarget :type Actor%20Object%20Reference) (Parameters :type Gameplay%20Cue%20Parameters%20Structure%20%28by%20ref%29) (ParticleComponents :type Array%20of%20FXSystem%20Component%20Object%20References) (AudioComponents :type Array%20of%20Audio%20Component%20Object%20References))
  (|CallParentFunction MyTarget Parameters ParticleComponents AudioComponents)
  (bind _returnvalue (Audio|SpawnSoundAttached (Variables|Default|GetFreeFallMusicTrack) (Utilities|Array|Get(acopy) ParticleComponents)))
  (bind _output_get (Variables|Default|SetMusic _returnvalue))
  (Audio|AddEventbyComponent (GameInstanceSubsystems|GetMusicEventSubsystem MyTarget) _output_get "(TagName=\"MusicEventSubsystem.Event.POI.Special\")" "(GameplayTags=)" 0 ""))

(fn OnRemoval(EitherVFX-preferthisversion)
  (params (MyTarget :type Actor%20Object%20Reference) (Parameters :type Gameplay%20Cue%20Parameters%20Structure%20%28by%20ref%29) (ParticleComponents :type Array%20of%20FXSystem%20Component%20Object%20References) (AudioComponents :type Array%20of%20Audio%20Component%20Object%20References) (BurstCameraShakeInstance :type Camera%20Shake%20Base%20Object%20Reference) (BurstDecalInstance :type Decal%20Actor%20Object%20Reference))
  (|CallParentFunction MyTarget Parameters ParticleComponents AudioComponents BurstCameraShakeInstance BurstDecalInstance)
  (bind _music (Variables|Default|GetMusic)
    (:then
      (Audio|RemoveEvent (GameInstanceSubsystems|GetMusicEventSubsystem MyTarget) _music))
    (:else)))

(fn ConstructionScript
  (|CallParentFunction))
