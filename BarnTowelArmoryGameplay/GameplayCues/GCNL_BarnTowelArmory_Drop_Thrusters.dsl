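; Looping gameplay cue for the thrusters on the BarnTowelArmory drop, adding sound indicator and audio mix handling to the parent cue's effects.
; On application, it shows a sound indicator at a valid target's location for the thruster landing-sequence sound, timed to match the burst sound.
; It also activates a control bus mix, then deactivates it after a brief delay rather than on removal.
; On removal, it stops the looping sound indicator.
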(event EventOnApplication(EitherVFX-preferthisone)
  (params (MyTarget :type Actor%20Object%20Reference) (Parameters :type Gameplay%20Cue%20Parameters%20Structure%20%28by%20ref%29) (ParticleComponents :type Array%20of%20FXSystem%20Component%20Object%20References) (AudioComponents :type Array%20of%20Audio%20Component%20Object%20References) (BurstCameraShakeInstance :type Camera%20Shake%20Base%20Object%20Reference) (BurstDecalInstance :type Decal%20Actor%20Object%20Reference))
  (bind _controlbusmix (Variables|Default|GetControlBusMix))
  (|CallParentFunction MyTarget Parameters ParticleComponents AudioComponents BurstCameraShakeInstance BurstDecalInstance)
  (Utilities|IsValid MyTarget
    (:Is%20Valid
      (SoundIndicator|ShowSoundIndicatorforDuration (Variables|Default|GetFortSoundIndicator) (Transformation|GetActorLocation MyTarget) "/BarnTowelArmoryGameplay/Assets/Sounds/SC_BarnTowelArmory_Thrusters_LandingSequence.SC_BarnTowelArmory_Thrusters_LandingSequence" 1.0 13.0 0) ; Play the same duration as the burst sound
    )
    (:Is%20Not%20Valid
      _))
  (Audio|Modulation|ActivateControlBusMix MyTarget _controlbusmix)
  (Utilities|FlowControl|Delay MyTarget 0.5)
  (Audio|Modulation|DeactivateControlBusMix MyTarget _controlbusmix))

(event EventOnRemoval(EitherVFX-preferthisversion)
  (params (MyTarget :type Actor%20Object%20Reference) (Parameters :type Gameplay%20Cue%20Parameters%20Structure%20%28by%20ref%29) (ParticleComponents :type Array%20of%20FXSystem%20Component%20Object%20References) (AudioComponents :type Array%20of%20Audio%20Component%20Object%20References) (BurstCameraShakeInstance :type Camera%20Shake%20Base%20Object%20Reference) (BurstDecalInstance :type Decal%20Actor%20Object%20Reference))
  (|CallParentFunction MyTarget Parameters ParticleComponents AudioComponents BurstCameraShakeInstance BurstDecalInstance)
  (SoundIndicator|StopSoundIndicatorLooping (Variables|Default|GetFortSoundIndicator)))

(fn ConstructionScript
  (|CallParentFunction))
