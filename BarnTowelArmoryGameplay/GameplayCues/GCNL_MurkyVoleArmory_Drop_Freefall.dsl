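; Looping gameplay cue for the MurkyVoleArmory drop's freefall phase that adds a sound indicator for its looping audio.
; On looping start, it calls the parent, then, if the target is valid and looping sounds exist, starts a targeted indicator using the first entry's sound cue.
; On removal, it calls the parent and stops the looping sound indicator.
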(event EventOnLoopingStart(EitherVFX-preferthisversion)
  (params (MyTarget :type Actor%20Object%20Reference) (Parameters :type Gameplay%20Cue%20Parameters%20Structure%20%28by%20ref%29) (ParticleComponents :type Array%20of%20FXSystem%20Component%20Object%20References) (AudioComponents :type Array%20of%20Audio%20Component%20Object%20References))
  (bind (_loopingsoundfadeoutduration _loopingsoundvolumelevel _condition _attachment _soundcue) (Utilities|Struct|BreakFortGameplayCueAudioInfoLooping (Utilities|Array|Get(aref) (Variables|LoopingEffects|GetLoopingSounds))))
  (bind _loopingsounds (Variables|LoopingEffects|GetLoopingSounds))
  (|CallParentFunction MyTarget Parameters ParticleComponents AudioComponents)
  (Utilities|IsValid MyTarget
    (:Is%20Valid
      (if (> (Utilities|Array|Length _loopingsounds) 0)
        (SoundIndicator|StartTargetedSoundIndicatorLooping (Variables|Default|GetFortSoundIndicator) MyTarget _soundcue 1.0 0)))))

(event EventOnRemoval(EitherVFX-preferthisversion)
  (params (MyTarget :type Actor%20Object%20Reference) (Parameters :type Gameplay%20Cue%20Parameters%20Structure%20%28by%20ref%29) (ParticleComponents :type Array%20of%20FXSystem%20Component%20Object%20References) (AudioComponents :type Array%20of%20Audio%20Component%20Object%20References) (BurstCameraShakeInstance :type Camera%20Shake%20Base%20Object%20Reference) (BurstDecalInstance :type Decal%20Actor%20Object%20Reference))
  (|CallParentFunction MyTarget Parameters ParticleComponents AudioComponents BurstCameraShakeInstance BurstDecalInstance)
  (SoundIndicator|StopSoundIndicatorLooping (Variables|Default|GetFortSoundIndicator)))

(fn ConstructionScript
  (|CallParentFunction))
