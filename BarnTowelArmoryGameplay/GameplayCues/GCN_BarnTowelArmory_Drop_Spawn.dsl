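; Burst gameplay cue for the BarnTowel Armory drop spawn that adds a local on-screen sound indicator to the spawn effects.
; OnBurst runs the parent burst, then, if the target is valid and burst sounds exist, shows the indicator with the BarnTowel icon and the first sound's cue.
; The indicator tracks the target actor to avoid ghosting against the engine loop playing simultaneously, with its attenuation range scaled down.
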(fn OnBurst(EitherVFX-preferthisversion)
  (params (MyTarget :type Actor%20Object%20Reference) (Parameters :type Gameplay%20Cue%20Parameters%20Structure%20%28by%20ref%29) (ParticleComponents :type Array%20of%20FXSystem%20Component%20Object%20References) (AudioComponents :type Array%20of%20Audio%20Component%20Object%20References) (BurstCameraShakeInstance :type Camera%20Shake%20Base%20Object%20Reference) (BurstDecalInstance :type Decal%20Actor%20Object%20Reference))
  (bind _bursteffectdata (Variables|BurstEffects|GetBurstEffectData))
  (bind (_burstparticles _burstsounds) (Utilities|Struct|BreakFortBurstEffectData _bursteffectdata))
  (bind (_condition _attachment _soundcue) (Utilities|Struct|BreakFortGameplayCueAudioInfo (Utilities|Array|Get(aref) _burstsounds)))
  (|CallParentFunction MyTarget Parameters ParticleComponents AudioComponents BurstCameraShakeInstance BurstDecalInstance)
  (Utilities|IsValid MyTarget
    (:Is%20Valid
      (if (> (Utilities|Array|Length _burstsounds) 0)
        (Fortnite|ShowSoundIndicatorLocal :WorldContextObject MyTarget :TrackedActor MyTarget :Location "0, 0, 0" :MaxAudibleDistance 0.0 :IndicatorType "Generic" :Tint "(R=0.000000,G=0.000000,B=0.000000,A=0.000000)" :OverrideIcon "/BarnTowelArmoryGameplay/Assets/Icons/T_UI_AV_BarnTowel.T_UI_AV_BarnTowel" :SoundBase _soundcue :AttenuationMaxRangeMultiplier 0.77 :AttenuationIntensityMultiplier 1.0) ; Follow armory so it doesn't get a ghosting effect\ndue to the engine loop playing at the same time
        (else
          _))
      (return))
    (:Is%20Not%20Valid
      _)))
